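# Niagara system for the holographic Crew pickaxe trail cosmetic, drawn by a single ribbon emitter.
# Anim-trail modules spawn, initialize and update short-lived particles that taper in width and fade in color and alpha over their life.
# Emitter-age curves also cut off the alpha and animate a dynamic material parameter, timing the look to the trail's age.
# Time-of-day scalars dim color and opacity at night, morning and evening.
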NiagaraSystem /BRCosmetics/Effects/Niagara/Systems/CosmeticTemplates/Pickaxe_Impacts/Crew/Holo/NS_Pickaxe_Crew_Trail_Holo
dependencies {
 usedDataInterfaces [&NiagaraDataInterfaceCurve, &NiagaraDataInterfaceSocketReader]
 usedDynamicInputs [&MakeFloatFromVector, &RotateVectorByQuaternion]
 usedModules [
  &self:SystemSpawnScript.SystemScriptSource.SystemScriptGraph.NiagaraNodeAssignment_0.SetVariables_#1
  &SystemState
 ]
}
emitters {Trail {
  emitterUpdateScript [
   &EmitterState()
   &FN_SpawnAnimTrail()
  ]
  particleSpawnScript [
   &FN_InitializeAnimTrailBoneIndecies()
   &FN_InitializeAnimTrail(AnimTrailWidth=0.85, Color=<rgba 1 1 1 0.5>, Lifetime=0.25)
   &SystemLocation(MaskSpawnGroup=false)
   &DynamicMaterialParameters(Write Parameter Index 0=false)
  ]
  particleUpdateScript [
   &ScaleRibbonWidth(Ribbon Width Scale=&FloatFromCurve(FloatCurve=DataInterface() {propertyValues {Curve {keys [{tangentMode=RCTM_User, time=0, value=1}, {tangentMode=RCTM_User, time=1, value=0}]}}}))
   &ScaleColor(
    Scale Alpha=&Multiply_Float(
     A=&FloatFromCurve(
      FloatCurve=DataInterface() {propertyValues {Curve {keys [
          {arriveTangent=-0.0134, interpMode=RCIM_Cubic, leaveTangent=-0.0134, tangentMode=RCTM_User, time=0.5, value=1}
          {arriveTangent=-4.1529, interpMode=RCIM_Cubic, leaveTangent=-4.1529, tangentMode=RCTM_User, time=1, value=0}
         ]}}}
      Scale Curve=&FloatFromCurve(
       FloatCurve=DataInterface() {propertyValues {Curve {keys [
           {arriveTangent=-1.7677, interpMode=RCIM_Cubic, leaveTangent=-1.7677, tangentMode=RCTM_User, time=0.1, value=1}
           {arriveTangent=-8.9897, interpMode=RCIM_Cubic, leaveTangent=-8.9897, tangentMode=RCTM_User, time=0.27, value=0}
          ]}}}
       CurveIndex=$Emitter.Age
      )
     )
     B=&FNTimeOfDayScalar(Night Scale=0.5, Morning Scale=0.75, Evening Scale=0.75)
    )
    Scale RGB=&VectorFromFloat(Value=&FloatFromCurve(
      FloatCurve=DataInterface() {propertyValues {Curve {keys [
          {interpMode=RCIM_Cubic, tangentMode=RCTM_User, time=0, value=1}
          {arriveTangent=-1.5604, interpMode=RCIM_Cubic, leaveTangent=-1.5604, tangentMode=RCTM_User, time=1, value=0}
         ]}}}
      Scale Curve=&FNTimeOfDayScalar(Night Scale=0.8, Morning Scale=0.9, Evening Scale=0.9)
     ))
    ScaleA=true
    ScaleRGB=true
   ) {module=ScaleColor002}
   &DynamicMaterialParameters(
    Index 1 Param 1=&FloatFromCurve(
     FloatCurve=DataInterface() {propertyValues {Curve {keys [{arriveTangent=-0.4165, interpMode=RCIM_Cubic, time=0.1, value=-0.875}, {arriveTangent=12.7593, leaveTangent=12.7593, time=0.3, value=0.5}]}}}
     CurveIndex=$Emitter.Age
    )
    Module.Index1Param0Write=true
    Module.Index1Param1Write=true
    Module.Index1Param2Write=true
    Module.Index1Param3Write=true
    Write Parameter Index 0=false
    Write Parameter Index 1=true
   ) {module=DynamicMaterialParameters001}
   &FN_UpdateAnimTrail()
  ]
  rendererClasses [&NiagaraRibbonRendererProperties]
  renderers [{rendererClass=&NiagaraRibbonRendererProperties}]
 }}
summary {
 systemName=NS_Pickaxe_Crew_Trail_Holo
 emitters [{emitterName=Trail, rendererClasses [&NiagaraRibbonRendererProperties]}]
 userVariables [
  User.bIsFrontEnd {type=NiagaraBool}
  User.bIsFrontEndPreview {type=NiagaraBool}
  User.RibbonIDOffset {type=NiagaraInt32}
 ]
}
